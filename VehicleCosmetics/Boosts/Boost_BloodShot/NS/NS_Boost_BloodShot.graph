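# Niagara system for the BloodShot vehicle boost cosmetic, with nozzle effects and flame trails that support painted color variants.
# Left and right emitters burst looping mesh and sprite particles at skeletal mesh sockets, rotating in opposite directions.
# A ribbon emitter spawns per-port trails that inherit fading source movement and darken with age.
# Its three renderers accompany age-driven big and small flame values and a time-stepped offset computed in inline HLSL.
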NiagaraSystem /VehicleCosmetics/Boosts/Boost_BloodShot/NS/NS_Boost_BloodShot
dependencies {
 usedDataInterfaces [
  &NiagaraDataInterfaceCurve
  &NiagaraDataInterfaceArrayFloat
  &NiagaraDataInterfaceColorCurve
  &NiagaraDataInterfaceSkeletalMesh
 ]
 usedDynamicInputs []
 usedModules [
  &self:SystemSpawnScript.SystemScriptSource.SystemScriptGraph.NiagaraNodeAssignment_0.SetVariables_#1
  &SystemState
 ]
}
emitters {
 mesh_Left001 {
  emitterUpdateScript [
   &EmitterState(
    Inactive Response=::NewEnumerator1
    Life Cycle Mode=::NewEnumerator1
    Loop Behavior=::NewEnumerator1
    Loop Duration Mode=::NewEnumerator1
    UseLoopDelay=false
   )
   &VC_FilteredSocketCount()
   &SpawnBurst_Instantaneous(Use Spawn Probability=false, UseLoopCountLimit=false)
   &SpawnBurst_Instantaneous(Spawn Group=1, Use Spawn Probability=false, UseLoopCountLimit=false) {module=SpawnBurst_Instantaneous001}
   &SpawnRate(SpawnRate=2.2, Use Spawn Probability=false) {enabled=false}
  ]
  particleSpawnScript [
   /SetVariables(Particles.VisibilityTag=$Emitter.SpawnGroup) {isSetParameters=true, module=SetVariables_#2}
   &VC_PaintedVariantsOverride(
    Black=<rgba 0 0 0 1.5>
    BurntSienna=<rgba 0.0642 0.0198 0.0079 1>
    Cobalt=<rgba 0 0.0483 0.2899 1>
    Crimson=<rgba 0.2899 0 0.0079 1>
    DefaultColor=<rgba 0.1267 0 0 1>
    ForestGreen=<rgba 0 0.2587 0.0131 1>
    Grey=<rgba 0.1059 0.1059 0.1059 1>
    Lime=<rgba 0.2196 0.4392 0 1>
    Orange=<rgba 0.3038 0.0549 0 1>
    Pink=<rgba 0.6858 0.1322 0.4845 1>
    Purple=<rgba 0.1432 0 0.349 1>
    R1 BurntSienna:LinearColor=<rgba 0.283 0.092 0.0408 1>
    R1 Cobalt:LinearColor=<rgba 0.151 0.3873 1 1>
    R1 Crimson:LinearColor=<rgba 0.4635 0.0974 0.1073 1>
    R1 DefaultColor=<rgba 1 0.5478 0.2 1>
    R1 ForestGreen:LinearColor=<rgba 0.1635 0.4149 0.1762 1>
    R1 Grey:LinearColor=<rgba 0.2587 0.2587 0.2587 1>
    R1 Lime:LinearColor=<rgba 0.7042 1 0 1>
    R1 Orange:LinearColor=<rgba 0.6724 0.5219 0 1>
    R1 Pink:LinearColor=<rgba 1 0.6684 0.8794 1>
    R1 Purple:LinearColor=<rgba 0.7602 0.0538 1 1>
    R1 Saffron:LinearColor=<rgba 1 0.7667 0 1>
    R1 SkyBlue:LinearColor=<rgba 0.1719 0.7229 1 1>
    R1 TitaniumWhite:LinearColor=<rgba 0.8906 0.8906 0.8906 1>
    Saffron=<rgba 0.1406 0.1319 0 1>
    SkyBlue=<rgba 0.0475 0.1349 0.1788 1>
    TitaniumWhite=<rgba 0.2587 0.2587 0.2587 1>
    Use R1 Painted Overrides=true
   )
   &InitializeParticle(
    Color=$Particles.Color
    Mesh Scale=&MergeHelper_VectorSwitch(0=<7 20 5>, 1=<6.9 20 4.9>, 2=?, 3=?, Index=$Emitter.SpawnGroup, Num=::NewEnumerator0)
    Mesh Scale Mode=::NewEnumerator4
    Sprite Rotation Mode=::NewEnumerator2
    UsePositionOffset=false
    Write Material Random=true
   )
   &PartitionParticles(Number Of Partitions=$Emitter.PortCount)
   &SkeletalMeshLocation(
    Bone / Socket Index=1
    Bone / Socket Sampling Mode=::NewEnumerator5
    MaskSpawnGroup=false
    Mesh Sampling Type=::NewEnumerator6
    Sampled Position Offset=&AddVectorToPosition(Position=<-45 0 0>, Vector=$User.DTCarOffset)
    Skeletal Mesh=$System.SkeletalMesh
   )
   &AddVelocity(Velocity=<0 0 0>) {enabled=false}
   &InitialMeshOrientation(Mesh Axis To Orient=<0 0 0>, Mesh Orientation Mode=::NewEnumerator2, Orientation Vector=<0 -1 0>, Use Rotation Vector=false)
  ]
  particleUpdateScript [
   &ParticleState(Kill Particles When Lifetime Has Elapsed=false, Loop Particles Lifetime=true)
   &ScaleMeshSize(Scale Factor=&VectorFromFloat(Value=0.5))
   &UpdateMeshOrientation(Rotation Vector=<0 1 0>)
   &SolveForcesAndVelocity(Clamp Velocity=false, Limit Acceleration=false)
  ]
  rendererClasses [&NiagaraMeshRendererProperties, &NiagaraSpriteRendererProperties]
  renderers [
   {rendererClass=&NiagaraMeshRendererProperties, rendererIndex=0}
   {rendererClass=&NiagaraMeshRendererProperties, rendererIndex=1}
   {rendererClass=&NiagaraSpriteRendererProperties, rendererIndex=2}
  ]
 }
 mesh_Right002 {
  emitterUpdateScript [
   &EmitterState(
    Inactive Response=::NewEnumerator1
    Life Cycle Mode=::NewEnumerator1
    Loop Behavior=::NewEnumerator1
    Loop Duration Mode=::NewEnumerator1
    UseLoopDelay=false
   )
   &VC_FilteredSocketCount()
   &SpawnBurst_Instantaneous(Use Spawn Probability=false, UseLoopCountLimit=false)
   &SpawnBurst_Instantaneous(Spawn Group=1, Use Spawn Probability=false, UseLoopCountLimit=false) {module=SpawnBurst_Instantaneous001}
   &SpawnRate(SpawnRate=2.2, Use Spawn Probability=false) {enabled=false}
  ]
  particleSpawnScript [
   /SetVariables(Particles.VisibilityTag=$Emitter.SpawnGroup) {isSetParameters=true, module=SetVariables_#2}
   &VC_PaintedVariantsOverride(
    Black=<rgba 0 0 0 1.5>
    BurntSienna=<rgba 0.0642 0.0198 0.0079 1>
    Cobalt=<rgba 0 0.0483 0.2899 1>
    Crimson=<rgba 0.2899 0 0.0079 1>
    DefaultColor=<rgba 0.1267 0 0 1>
    ForestGreen=<rgba 0 0.2587 0.0131 1>
    Grey=<rgba 0.1059 0.1059 0.1059 1>
    Lime=<rgba 0.2196 0.4392 0 1>
    Orange=<rgba 0.3038 0.0549 0 1>
    Pink=<rgba 0.6858 0.1322 0.4845 1>
    Purple=<rgba 0.1432 0 0.349 1>
    R1 BurntSienna:LinearColor=<rgba 0.283 0.092 0.0408 1>
    R1 Cobalt:LinearColor=<rgba 0.151 0.3873 1 1>
    R1 Crimson:LinearColor=<rgba 0.4635 0.0974 0.1073 1>
    R1 DefaultColor=<rgba 1 0.5478 0.2 1>
    R1 ForestGreen:LinearColor=<rgba 0.1635 0.4149 0.1762 1>
    R1 Grey:LinearColor=<rgba 0.2587 0.2587 0.2587 1>
    R1 Lime:LinearColor=<rgba 0.7042 1 0 1>
    R1 Orange:LinearColor=<rgba 0.6724 0.5219 0 1>
    R1 Pink:LinearColor=<rgba 1 0.6684 0.8794 1>
    R1 Purple:LinearColor=<rgba 0.7602 0.0538 1 1>
    R1 Saffron:LinearColor=<rgba 1 0.7667 0 1>
    R1 SkyBlue:LinearColor=<rgba 0.1719 0.7229 1 1>
    R1 TitaniumWhite:LinearColor=<rgba 0.8906 0.8906 0.8906 1>
    Saffron=<rgba 0.1406 0.1319 0 1>
    SkyBlue=<rgba 0.0475 0.1349 0.1788 1>
    TitaniumWhite=<rgba 0.2587 0.2587 0.2587 1>
    Use R1 Painted Overrides=true
   )
   &InitializeParticle(
    Color=$Particles.Color
    Mesh Scale=&MergeHelper_VectorSwitch(0=<7 20 5>, 1=<6.9 20 4.9>, 2=?, 3=?, Index=$Emitter.SpawnGroup, Num=::NewEnumerator0)
    Mesh Scale Mode=::NewEnumerator4
    Sprite Rotation Mode=::NewEnumerator2
    UsePositionOffset=false
    Write Material Random=true
   )
   &PartitionParticles(Number Of Partitions=$Emitter.PortCount)
   &SkeletalMeshLocation(
    Bone / Socket Sampling Mode=::NewEnumerator5
    MaskSpawnGroup=false
    Mesh Sampling Type=::NewEnumerator6
    Sampled Position Offset=&AddVectorToPosition(Position=<-45 0 0>, Vector=$User.DTCarOffset)
    Skeletal Mesh=$System.SkeletalMesh
   )
   &AddVelocity(Velocity=<0 0 0>) {enabled=false}
   &InitialMeshOrientation(Mesh Axis To Orient=<0 0 0>, Mesh Orientation Mode=::NewEnumerator2, Orientation Vector=<0 -1 0>, Use Rotation Vector=false)
  ]
  particleUpdateScript [
   &ParticleState(Kill Particles When Lifetime Has Elapsed=false, Loop Particles Lifetime=true)
   &ScaleMeshSize(Scale Factor=&VectorFromFloat(Value=0.5))
   &UpdateMeshOrientation(Rotation Vector=<0 -1 0>)
   &SolveForcesAndVelocity(Clamp Velocity=false, Limit Acceleration=false)
  ]
  rendererClasses [&NiagaraMeshRendererProperties, &NiagaraSpriteRendererProperties]
  renderers [
   {rendererClass=&NiagaraMeshRendererProperties, rendererIndex=0}
   {rendererClass=&NiagaraMeshRendererProperties, rendererIndex=1}
   {rendererClass=&NiagaraSpriteRendererProperties, rendererIndex=2}
  ]
 }
 ribbon {
  emitterUpdateScript [
   &EmitterState(Inactive Response=::NewEnumerator1, Life Cycle Mode=::NewEnumerator1, Loop Duration=1, UseLoopDelay=false)
   &VC_FilteredSocketCount()
   &VC_RateQualityHandler(Epic Spawn Rate=45, Low Spawn Rate=30, Use Custom Rate?=true) {enabled=false}
   &VC_SpawnRibbon_v1(Number of Ribbons=$Emitter.PortCount, SpawnRate=60)
   /SetVariables(Emitter.Offset X=HlslExpression() {hlslExpression="floor(fmod(Engine.Time * -14, 12)) / 8"}) {isSetParameters=true}
  ]
  particleSpawnScript [
   &VC_PaintedVariantsOverride(
    BurntSienna=<rgba 0.4358 0.2133 0.1536 1>
    Crimson=<rgba 0.717 0 0.0194 1>
    DefaultColor=<rgba 1 0.2 0.2476 1>
    Grey=<rgba 0.4323 0.4323 0.4323 1>
    Lime=<rgba 0.46 0.6858 0.0774 1>
    Orange=<rgba 0.85 0.2965 0 1>
    Pink=<rgba 1.3 0.5033 1.0103 1>
    Purple=<rgba 0.3689 0.1073 0.7448 1>
    R1 Black:LinearColor=<rgba 0.1128 0.1128 0.1128 1>
    R1 BurntSienna:LinearColor=<rgba 0.6476 0.225 0.0971 1>
    R1 Cobalt:LinearColor=<rgba 0.3126 0.5317 0.7795 1>
    R1 Crimson:LinearColor=<rgba 0.7969 0.1314 0.1495 1>
    R1 DefaultColor=<rgba 1 0.6326 0.35 1>
    R1 ForestGreen:LinearColor=<rgba 0.1669 0.8663 0.2023 1>
    R1 Grey:LinearColor=<rgba 0.533 0.533 0.533 1>
    R1 Lime:LinearColor=<rgba 1 0.9645 0 1>
    R1 Orange:LinearColor=<rgba 0.6724 0.5941 0 1>
    R1 Pink:LinearColor=<rgba 1 0.6649 0.8782 1>
    R1 Purple:LinearColor=<rgba 0.783 0.1482 0.7672 1>
    R1 Saffron:LinearColor=<rgba 1 0.7667 0 1>
    R1 SkyBlue:LinearColor=<rgba 0.3385 0.7787 1 1>
    R1 TitaniumWhite:LinearColor=<rgba 1 1 1 1>
    R2 BurntSienna:LinearColor=<rgba 0.6267 0.2844 0.1926 1>
    R2 Cobalt:LinearColor=<rgba 0.4253 0.6497 1 1>
    R2 Crimson:LinearColor=<rgba 1 0.2726 0.2923 1>
    R2 DefaultColor:LinearColor=<rgba 1 0.6935 0.2066 1>
    R2 ForestGreen:LinearColor=<rgba 0.2565 0.7899 0.2834 1>
    R2 Grey:LinearColor=<rgba 0.6372 0.6372 0.6372 1>
    R2 Lime:LinearColor=<rgba 0.5551 0.8003 0.3099 1>
    R2 Pink:LinearColor=<rgba 1 0.6372 0.8681 1>
    R2 Purple:LinearColor=<rgba 0.6018 0.3247 1 1>
    R2 SkyBlue:LinearColor=<rgba 0.4462 0.8147 1 1>
    R2 TitaniumWhite:LinearColor=<rgba 1 1 1 1>
    Saffron=<rgba 0.6858 0.6495 0.1012 1>
    SkyBlue=<rgba 0.1044 0.4843 0.6753 1>
    TitaniumWhite=<rgba 0.526 0.526 0.526 1>
    Use R1 Painted Overrides=true
    Use R2 Painted Overrides=true
   )
   &VC_InitializeRibbon_v1(
    Color=$Particles.Color
    Custom Offset=true
    Enable Skeletal Position=true
    Lifetime=0.9
    Local Vector Offset=&AddVectorToPosition(Position=<-30 0 0>, Vector=$User.DTCarOffset)
    Offset Amount=1
    Ribbon Width=40
    Write Material Random=false
   )
   &AddVelocity(Velocity=<1800 0 0>, Velocity Speed Scale=-1) {module=AddVelocity001}
  ]
  particleUpdateScript [
   &ParticleState()
   &InheritSourceMovement(
    Applied Position Scale=&Multiply_VectorByFloat(
     Vector=&VectorFromFloat(Value=&FloatFromCurve(FloatCurve=DataInterface() {propertyValues {Curve {
          defaultValue=0
          keys [
           {arriveTangent=-0.1936, interpMode=RCIM_Cubic, leaveTangent=-0.1936, tangentMode=RCTM_User, time=0, value=1}
           {arriveTangent=-0.1381, interpMode=RCIM_Cubic, leaveTangent=-0.1381, tangentMode=RCTM_User, time=1, value=0}
          ]
         }}}))
     Float=1
    )
    ApplyForce=false
    ApplyPosition=true
    ApplyPositionFalloff=false
   )
   &SolveForcesAndVelocity(Clamp Velocity=false, Limit Acceleration=false)
   &VC_UpdateRibbon_v1(Enable Skeletal Position=true)
   /SetVariables(
    Particles.BigFlame:NiagaraFloat=&FloatFromCurve(
     FloatCurve=DataInterface() {propertyValues {Curve {
        defaultValue=1
        keys [
         {arriveTangent=6.7676, interpMode=RCIM_Cubic, leaveTangent=6.7676, tangentMode=RCTM_User, time=0, value=0}
         {arriveTangent=1.3937, interpMode=RCIM_Cubic, leaveTangent=1.3937, tangentMode=RCTM_User, time=0.3, value=1.0633}
         {interpMode=RCIM_Cubic, leaveTangent=0.2857, time=1, value=2}
        ]
       }}}
     Scale Curve=&SelectFloatFromArray(Direct Array Index=&SpawnGroup(), Float Selection Array=DataInterface() {propertyValues {FloatData [60,60]}})
    )
    Particles.SmallFlame:NiagaraFloat=&FloatFromCurve(
     FloatCurve=DataInterface() {propertyValues {Curve {
        defaultValue=1
        keys [
         {arriveTangent=9.6637, interpMode=RCIM_Cubic, leaveTangent=9.6637, tangentMode=RCTM_User, time=0, value=0}
         {arriveTangent=-3.2941, interpMode=RCIM_Cubic, leaveTangent=-3.2941, tangentMode=RCTM_User, time=0.4, value=1}
         {arriveTangent=-1.6482, tangentMode=RCTM_User, time=1, value=0}
        ]
       }}}
     Scale Curve=&SelectFloatFromArray(Direct Array Index=&SpawnGroup(), Float Selection Array=DataInterface() {propertyValues {FloatData [60,60]}})
    )
    Particles.YellowFlame=70
   ) {isSetParameters=true}
   /SetVariables(Particles.WPO:NiagaraPosition=&AddVectorToPosition(
     Position=$Particles.Position
     Vector=&TransformVector(
      Vector=&MakeVector(
       X=0
       Y=0
       Z=&Sine(
        Normalized Angle=&Add_Float(
         A=&Multiply_Float(A=$Particles.NormalizedAge, B=-3)
         B=&Multiply_Float(A=$Emitter.LoopedAge, B=2)
        )
        Period=&MaskFloatBySpawnGroup(Masked Float=0.5, Passthrough Float=-0.5)
        Scale=&FloatFromCurve(
         FloatCurve=DataInterface() {propertyValues {Curve {
            defaultValue=1
            keys [
             {arriveTangent=-0.0532, interpMode=RCIM_Cubic, leaveTangent=-0.0532, tangentMode=RCTM_User, time=0, value=0}
             <at 0.5 1 RCIM_Cubic>
             {interpMode=RCIM_Cubic, tangentMode=RCTM_User, time=1, value=0}
            ]
           }}}
         Scale Curve=2
        )
        Bias=0
       )
      )
      Destination Space=::Simulation
      Apply Scale=true
     )
    )) {enabled=false, isSetParameters=true}
   &ScaleColor(
    Scale RGB=&MakeCustomVectorFromBool(
     Bool=$User.bIsPainted
     True Vector=&VectorFromFloat(Value=&FloatFromCurve(FloatCurve=DataInterface() {propertyValues {Curve {keys [
           {arriveTangent=-1.5315, interpMode=RCIM_Cubic, leaveTangent=-1.5315, tangentMode=RCTM_User, time=0, value=1}
           {arriveTangent=0.015, interpMode=RCIM_Cubic, leaveTangent=0.015, tangentMode=RCTM_User, time=1, value=0.2}
          ]}}}))
     False Vector=&VectorFromFloat(Value=&FloatFromCurve(FloatCurve=DataInterface() {propertyValues {Curve {keys [{arriveTangent=-1.6351, interpMode=RCIM_Cubic, leaveTangent=-1.6351, tangentMode=RCTM_User, time=0, value=1}, <at 0.6 0.25 RCIM_Cubic>]}}}))
    )
    ScaleA=true
    ScaleRGB=true
   )
  ]
  rendererClasses [&NiagaraRibbonRendererProperties]
  renderers [
   {rendererClass=&NiagaraRibbonRendererProperties}
   {rendererClass=&NiagaraRibbonRendererProperties, rendererIndex=1}
   {rendererClass=&NiagaraRibbonRendererProperties, rendererIndex=2}
  ]
 }
}
summary {
 systemName=NS_Boost_BloodShot
 emitters [
  {emitterName=mesh_Right002, rendererClasses [&NiagaraMeshRendererProperties, &NiagaraSpriteRendererProperties]}
  {emitterName=mesh_Left001, rendererClasses [&NiagaraMeshRendererProperties, &NiagaraSpriteRendererProperties]}
  {emitterName=ribbon, rendererClasses [&NiagaraRibbonRendererProperties]}
 ]
 userVariables [
  User.bIsFrontendPreview {type=NiagaraBool}
  User.bIsPainted {type=NiagaraBool}
  User.ColorVariantID {type=NiagaraInt32}
  User.DTCarOffset {type=Vector3f, description="", defaultValue {struct=Vector3f, value=<0 0 0>}}
  User.PaintedColor {type=LinearColor, description="", defaultValue {struct=LinearColor, value=<rgba 1 1 1 1>}}
  User.PaintedScalar {type=NiagaraFloat, defaultValue=1, description=""}
  User.PortCount {type=NiagaraInt32, defaultValue=2}
 ]
}
